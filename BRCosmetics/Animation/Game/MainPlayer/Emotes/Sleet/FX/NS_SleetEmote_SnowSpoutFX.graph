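# Niagara system for the Sleet emote's snow spout, throwing colliding snow, mesh chunks, smoke, spray and snowflakes from a sampled skeletal mesh.
# Particles launch with directional, cone and inherited velocity, then fall under gravity and drag or curl noise.
# Movement-based spawning stops in water and picks CPU emitter variants at the two lowest quality levels, GPU ones above.
# Front-end preview swaps this for fixed spawn rates.
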NiagaraSystem /BRCosmetics/Animation/Game/MainPlayer/Emotes/Sleet/FX/NS_SleetEmote_SnowSpoutFX
dependencies {
 usedDataInterfaces [
  &NiagaraDataInterfaceArrayFloat
  &NiagaraDataInterfaceCurve
  &NiagaraDataInterfaceColorCurve
  &NiagaraDataInterfaceSpriteRendererInfo
  &NiagaraDataInterfaceSkeletalMesh
 ]
 usedDynamicInputs [&RandomRangeVector2D]
 usedModules [
  &self:SystemSpawnScript.SystemScriptSource.SystemScriptGraph.NiagaraNodeAssignment_0.SetVariables_#1
  &SystemState
 ]
}
emitters {
 Collision_Snow {
  simTarget=GPUComputeSim
  emitterUpdateScript [
   &EmitterState()
   &SpawnPerUnit(
    Module.Use Max Movement Threshold=true
    Module.Use Movement Tolerance=true
    Movement Tolerance:NiagaraFloat=$User.MovementTolerance
    Spawn Per Unit:NiagaraFloat=&MakeCustomFloatFromBool(
     Bool=$User.bIsFrontEndPreview
     True Float=0
     False Float=&Multiply_Float(
      A=&MakeCustomFloatFromBool(Bool=$User.bIsInWater, True Float=0, False Float=1)
      B=&SelectFloatFromArray(Direct Array Index=$Engine.QualityLevel, Float Selection Array=DataInterface() {propertyValues {FloatData [0,0,20,15,12]}})
     )
    )
    Use Spawn Probability=false
   )
   &SpawnRate(
    SpawnRate=&MakeCustomFloatFromBool(Bool=$User.bIsFrontEndPreview, True Float=10, False Float=0)
    Use Spawn Probability=false
   ) {module=SpawnRate003}
  ]
  particleSpawnScript [
   &InitializeParticle(
    Lifetime=&UniformRangedFloat(Minimum=3, Maximum=4)
    Position=$Engine.Owner.Position
    Position Mode=::NewEnumerator1
    Sprite Rotation Angle=&UniformRangedFloat(Minimum=0, Maximum=360)
    Sprite Rotation Mode=::NewEnumerator0
    Sprite Size=&Vector2DFromFloat(Value=&UniformRangedFloat(Minimum=5, Maximum=8))
    Sprite Size Mode=::NewEnumerator4
    UsePositionOffset=false
    Write Material Random=true
   )
   &SkeletalMeshLocation(MaskSpawnGroup=false, Mesh Sampling Type=::NewEnumerator5, Skeletal Mesh=$System.Skeletal Mesh, Socket Sampling Mode=::NewEnumerator5)
   &SubUVAnimation(End Frame=3, SubUV Animation Mode:ENiagara_SubUVLookupMode=::NewEnumerator2)
   &ShapeLocation(MaskSpawnGroup=false, Override Seed=false, Sphere Radius=0, UseOwnerScale=true)
   &InheritVelocity(
    Module.Limit Speed=true
    Velocity Limit=50
    Velocity Scale=&RandomRangeVector(Minimum=<0.5 0.5 0.5>, Maximum=<1 1 1>)
   ) {module=InheritVelocity001}
   &AddVelocityInCone(
    Cone Angle=50
    Cone Axis=<0 1 1>
    Module.Use Velocity Falloff On Cone Axis=false
    Velocity Strength=&UniformRangedFloat(Minimum=25, Maximum=100)
   )
   &AddVelocity(Velocity=&RandomRangeVector(Minimum=<-100 350 150>, Maximum=<-300 500 250>)) {module=AddVelocity001}
  ]
  particleUpdateScript [
   &ParticleState()
   &ScaleSpriteSize(
    Uniform Curve Scale=&RandomRangeFloat(Minimum=0.75, Maximum=1.25)
    Uniform Curve Sprite Scale=DataInterface() {propertyValues {Curve {keys [
        {arriveTangent=-2.7542, interpMode=RCIM_Cubic, leaveTangent=-2.7542, tangentMode=RCTM_User, time=0, value=1}
        {tangentMode=RCTM_User, time=1, value=0}
       ]}}}
   ) {module=ScaleSpriteSize001}
   &DynamicMaterialParameters(
    Index 0 Param 1=&FloatFromCurve(FloatCurve=DataInterface() {propertyValues {Curve {keys [<at 0 -1>,<at 1 1>]}}})
    Module.Param0WriteEnabled=true
    Module.Param1WriteEnabled=true
    Module.Param2WriteEnabled=true
    Module.Param3WriteEnabled=true
   )
   &GravityForce(Gravity=&MakeVector(X=0, Y=0, Z=&Multiply_Float(A=-980, B=1.6))) {module=GravityForce001}
   &Drag(Use Linear Drag=true, Use Rotational Drag=true)
   &Collision(
    Age Colliding Particles=false
    Friction=1
    Friction During a Bounce=5
    GPU Collision Type=::NewEnumerator2
    Module.Control Rest State Via Surface Normal=false
    Module.EnableMaxCollisionCount=false
    Module.Randomize Collision Normal Vector=false
    Restitution=0
   )
   &SolveForcesAndVelocity(Clamp Velocity=false, Limit Acceleration=false)
   &ScaleColor(
    Scale Alpha=&FloatFromCurve(
     FloatCurve=DataInterface() {propertyValues {Curve {keys [
         {arriveTangent=4.1785, interpMode=RCIM_Cubic, leaveTangent=4.1785, tangentMode=RCTM_User, time=0, value=0}
         {arriveTangent=0.9995, interpMode=RCIM_Cubic, leaveTangent=0.9995, tangentMode=RCTM_User, time=0.0178, value=1}
         {arriveTangent=-1.0182, interpMode=RCIM_Cubic, leaveTangent=-1.0182, time=0.5124, value=0.3091}
         {arriveTangent=-0.3824, interpMode=RCIM_Cubic, leaveTangent=-0.0405, tangentMode=RCTM_Break, time=1, value=0}
        ]}}}
     Scale Curve=0.7
    )
    ScaleA=true
    ScaleRGB=true
   )
  ]
  rendererClasses [&NiagaraSpriteRendererProperties]
  renderers [{rendererClass=&NiagaraSpriteRendererProperties}]
 }
 Collision_Snow_CPU {
  emitterUpdateScript [
   &EmitterState()
   &SpawnPerUnit(
    Module.Use Max Movement Threshold=true
    Module.Use Movement Tolerance=true
    Movement Tolerance:NiagaraFloat=$User.MovementTolerance
    Spawn Per Unit:NiagaraFloat=&MakeCustomFloatFromBool(
     Bool=$User.bIsFrontEndPreview
     True Float=0
     False Float=&Multiply_Float(
      A=&MakeCustomFloatFromBool(Bool=$User.bIsInWater, True Float=0, False Float=1)
      B=&SelectFloatFromArray(Direct Array Index=$Engine.QualityLevel, Float Selection Array=DataInterface() {propertyValues {FloatData [40,30,0,0,0]}})
     )
    )
    Use Spawn Probability=false
   )
   &SpawnRate(
    SpawnRate=&MakeCustomFloatFromBool(Bool=$User.bIsFrontEndPreview, True Float=8, False Float=0)
    Use Spawn Probability=false
   ) {module=SpawnRate003}
  ]
  particleSpawnScript [
   &InitializeParticle(
    Lifetime=&UniformRangedFloat(Minimum=3, Maximum=4)
    Position=$Engine.Owner.Position
    Position Mode=::NewEnumerator1
    Sprite Rotation Angle=&UniformRangedFloat(Minimum=0, Maximum=360)
    Sprite Rotation Mode=::NewEnumerator0
    Sprite Size=&Vector2DFromFloat(Value=&UniformRangedFloat(Minimum=5, Maximum=8))
    Sprite Size Mode=::NewEnumerator4
    UsePositionOffset=false
    Write Material Random=true
   )
   &SkeletalMeshLocation(MaskSpawnGroup=false, Mesh Sampling Type=::NewEnumerator5, Skeletal Mesh=$System.Skeletal Mesh, Socket Sampling Mode=::NewEnumerator5)
   &SubUVAnimation(End Frame=3, SubUV Animation Mode:ENiagara_SubUVLookupMode=::NewEnumerator2)
   &ShapeLocation(MaskSpawnGroup=false, Override Seed=false, Sphere Radius=0, UseOwnerScale=true)
   &InheritVelocity(
    Module.Limit Speed=true
    Velocity Limit=50
    Velocity Scale=&RandomRangeVector(Minimum=<0.5 0.5 0.5>, Maximum=<1 1 1>)
   ) {module=InheritVelocity001}
   &AddVelocityInCone(
    Cone Angle=50
    Cone Axis=<0 1 1>
    Module.Use Velocity Falloff On Cone Axis=false
    Velocity Strength=&UniformRangedFloat(Minimum=25, Maximum=100)
   )
   &AddVelocity(Velocity=&RandomRangeVector(Minimum=<-100 350 150>, Maximum=<-200 500 250>)) {module=AddVelocity001}
  ]
  particleUpdateScript [
   &ParticleState()
   &ScaleSpriteSize(
    Uniform Curve Scale=&RandomRangeFloat(Minimum=0.75, Maximum=1.25)
    Uniform Curve Sprite Scale=DataInterface() {propertyValues {Curve {keys [
        {arriveTangent=-2.7542, interpMode=RCIM_Cubic, leaveTangent=-2.7542, tangentMode=RCTM_User, time=0, value=1}
        {tangentMode=RCTM_User, time=1, value=0}
       ]}}}
   ) {module=ScaleSpriteSize001}
   &DynamicMaterialParameters(
    Index 0 Param 1=&FloatFromCurve(FloatCurve=DataInterface() {propertyValues {Curve {keys [<at 0 -1>,<at 1 1>]}}})
    Module.Param0WriteEnabled=true
    Module.Param1WriteEnabled=true
    Module.Param2WriteEnabled=true
    Module.Param3WriteEnabled=true
   )
   &GravityForce(Gravity=&MakeVector(X=0, Y=0, Z=&Multiply_Float(A=-980, B=1.6))) {module=GravityForce001}
   &Drag(Use Linear Drag=true, Use Rotational Drag=true)
   &Collision(
    Age Colliding Particles=false
    Friction=1
    Friction During a Bounce=5
    Module.Control Rest State Via Surface Normal=false
    Module.EnableMaxCollisionCount=false
    Module.Randomize Collision Normal Vector=false
    Restitution=0
   )
   &SolveForcesAndVelocity(Clamp Velocity=false, Limit Acceleration=false)
   &ScaleColor(
    Scale Alpha=&FloatFromCurve(
     FloatCurve=DataInterface() {propertyValues {Curve {keys [
         {arriveTangent=4.1785, interpMode=RCIM_Cubic, leaveTangent=4.1785, tangentMode=RCTM_User, time=0, value=0}
         {arriveTangent=0.9995, interpMode=RCIM_Cubic, leaveTangent=0.9995, tangentMode=RCTM_User, time=0.0178, value=1}
         {arriveTangent=-1.0182, interpMode=RCIM_Cubic, leaveTangent=-1.0182, time=0.5124, value=0.3091}
         {arriveTangent=-0.3824, interpMode=RCIM_Cubic, leaveTangent=-0.0405, tangentMode=RCTM_Break, time=1, value=0}
        ]}}}
     Scale Curve=0.7
    )
    ScaleA=true
    ScaleRGB=true
   )
  ]
  rendererClasses [&NiagaraSpriteRendererProperties]
  renderers [{rendererClass=&NiagaraSpriteRendererProperties}]
 }
 Snow_Chunks {
  simTarget=GPUComputeSim
  emitterUpdateScript [
   &EmitterState()
   &SpawnPerUnit(
    Module.Use Max Movement Threshold=true
    Module.Use Movement Tolerance=true
    Movement Tolerance:NiagaraFloat=$User.MovementTolerance
    Spawn Per Unit:NiagaraFloat=&MakeCustomFloatFromBool(
     Bool=$User.bIsFrontEndPreview
     True Float=0
     False Float=&Multiply_Float(
      A=&MakeCustomFloatFromBool(Bool=$User.bIsInWater, True Float=0, False Float=1)
      B=&SelectFloatFromArray(Direct Array Index=$Engine.QualityLevel, Float Selection Array=DataInterface() {propertyValues {FloatData [0,0,2,1.5,1]}})
     )
    )
    Use Spawn Probability=false
   )
   &SpawnRate(
    SpawnRate=&MakeCustomFloatFromBool(Bool=$User.bIsFrontEndPreview, True Float=25, False Float=0)
    Use Spawn Probability=false
   )
  ]
  particleSpawnScript [
   /SetVariables(Particles.MeshIndex=&RandomRangeInt(Minimum=0, Maximum=1)) {isSetParameters=true, module=SetVariables_#2}
   /SetVariables(Particles.Intensity:NiagaraFloat=&FloatFromCurve(
     FloatCurve=DataInterface() {propertyValues {Curve {keys [<at 0.8 0>,<at 1 1>]}}}
     CurveIndex=&RandomRangeFloat(Minimum=0, Maximum=1)
    )) {isSetParameters=true, module=SetVariables_#3}
   &SkeletalMeshLocation(MaskSpawnGroup=false, Mesh Sampling Type=::NewEnumerator5, Skeletal Mesh=$System.Skeletal Mesh, Socket Sampling Mode=::NewEnumerator5)
   &InitializeParticle(
    Lifetime=&UniformRangedFloat(Minimum=0.15, Maximum=0.3)
    Mass=10
    Mass Mode=::NewEnumerator1
    Mesh Scale=&Multiply_VectorByFloat(
     Vector=&RandomRangeVector(Minimum=<1.5 0.5 0.5>, Maximum=<3 1.5 1.5>)
     Float=&Lerp_Float(A=0.04, B=0.06, Alpha=$Particles.Intensity)
    )
    Mesh Scale Mode=::NewEnumerator4
    Sprite Size Mode=::NewEnumerator4
    UsePositionOffset=false
    Write Material Random=true
   ) {module=InitializeParticle001}
   &AddVelocity(Velocity=&RandomRangeVector(Minimum=<0 100 150>, Maximum=<-200 400 250>))
   &InheritVelocity(
    Module.Limit Speed=true
    Velocity Limit=50
    Velocity Scale=&RandomRangeVector(Minimum=<0.5 0.5 0.5>, Maximum=<1 1 1>)
   )
   &AddVelocityInCone(
    Cone Angle=50
    Cone Axis=<0 1 1>
    Module.Use Velocity Falloff On Cone Axis=false
    Velocity Strength=&UniformRangedFloat(Minimum=0, Maximum=50)
   )
   /Niagara/Modules/Spawn/Location/SphereLocation(
    MaskSpawnGroup=false
    Module.Override Randomness=false
    Sphere Origin:NiagaraPosition=&ConvertVectorToPosition(
     Input Position=Linked() {linkedVariable {name=Particles.SkeletalMeshLocation.SampledPosition, type=NiagaraPosition}}
     Vector Is In Absolute Large World Space:NiagaraBool=false
    )
    Sphere Radius=2
    Surface Only=false
   ) {module=SphereLocation001}
  ]
  particleUpdateScript [
   &ParticleState()
   &Drag(Use Linear Drag=true, Use Rotational Drag=false)
   &GravityForce(Gravity=<0 0 -720>)
   &CurlNoiseForce(Noise Frequency=100, Noise Strength=50, Pan Noise=false)
   &SolveForcesAndVelocity(Clamp Velocity=false, Limit Acceleration=false)
   &ScaleColor(
    Scale Alpha=1.2
    Scale RGB=&MakeVectorFromLinearColorRGB(Color=&ColorFromCurve(ColorCurve=DataInterface() {propertyValues {
        AlphaCurve {keys []}
        BlueCurve {keys [<at 0.0081 0.6244>,<at 0.1423 0.6305>,<at 0.9559 1>]}
        GreenCurve {keys [<at 0.0081 0.8353>,<at 0.1423 0.7629>,<at 0.9559 0.8821>]}
        RedCurve {keys [<at 0.0081 1>,<at 0.1423 1>,<at 0.9559 0.8145>]}
       }}))
    ScaleA=true
    ScaleRGB=true
   )
   &UpdateMeshOrientation(Rotation Rate=&RandomRangeFloat(Minimum=-0.4, Maximum=0.4), Rotation Vector=<1 1 1>)
   &ScaleMeshSize(Scale Factor=&VectorFromFloat(Value=&FloatFromCurve(FloatCurve=DataInterface() {propertyValues {Curve {keys [
          {arriveTangent=8.5765, interpMode=RCIM_Cubic, leaveTangent=8.5765, tangentMode=RCTM_User, time=0.1, value=0.7}
          {arriveTangent=0.65, interpMode=RCIM_Cubic, leaveTangent=0.65, time=0.2, value=1}
          {arriveTangent=-1.25, interpMode=RCIM_Cubic, leaveTangent=-1.25, time=0.6, value=0.625}
          {arriveTangent=0.095, interpMode=RCIM_Cubic, leaveTangent=0.095, tangentMode=RCTM_User, time=1, value=0}
         ]}}})))
   &ScaleSpriteSizeBySpeed(
    Max Scale Factor=&Vector2DFromFloat(Value=1)
    Min Scale Factor=&Vector2DFromFloat(Value=0.25)
    Velocity Threshold=200
   )
  ]
  rendererClasses [&NiagaraMeshRendererProperties]
  renderers [{rendererClass=&NiagaraMeshRendererProperties, rendererIndex=0}]
 }
 Snow_Chunks_CPU {
  emitterUpdateScript [
   &EmitterState()
   &SpawnPerUnit(
    Module.Use Max Movement Threshold=true
    Module.Use Movement Tolerance=true
    Movement Tolerance:NiagaraFloat=$User.MovementTolerance
    Spawn Per Unit:NiagaraFloat=&MakeCustomFloatFromBool(
     Bool=$User.bIsFrontEndPreview
     True Float=0
     False Float=&Multiply_Float(
      A=&MakeCustomFloatFromBool(Bool=$User.bIsInWater, True Float=0, False Float=1)
      B=&SelectFloatFromArray(Direct Array Index=$Engine.QualityLevel, Float Selection Array=DataInterface() {propertyValues {FloatData [7,6,0,0,0]}})
     )
    )
    Use Spawn Probability=false
   )
   &SpawnRate(
    SpawnRate=&MakeCustomFloatFromBool(Bool=$User.bIsFrontEndPreview, True Float=12, False Float=0)
    Use Spawn Probability=false
   )
  ]
  particleSpawnScript [
   /SetVariables(Particles.MeshIndex=&RandomRangeInt(Minimum=0, Maximum=1)) {isSetParameters=true, module=SetVariables_#2}
   /SetVariables(Particles.Intensity:NiagaraFloat=&FloatFromCurve(
     FloatCurve=DataInterface() {propertyValues {Curve {keys [<at 0.8 0>,<at 1 1>]}}}
     CurveIndex=&RandomRangeFloat(Minimum=0, Maximum=1)
    )) {isSetParameters=true, module=SetVariables_#3}
   &SkeletalMeshLocation(MaskSpawnGroup=false, Mesh Sampling Type=::NewEnumerator5, Skeletal Mesh=$System.Skeletal Mesh, Socket Sampling Mode=::NewEnumerator5)
   &InitializeParticle(
    Lifetime=&UniformRangedFloat(Minimum=0.15, Maximum=0.3)
    Mass=10
    Mass Mode=::NewEnumerator1
    Mesh Scale=&Multiply_VectorByFloat(
     Vector=&RandomRangeVector(Minimum=<1.5 0.5 0.5>, Maximum=<3 1.5 1.5>)
     Float=&Lerp_Float(A=0.04, B=0.06, Alpha=$Particles.Intensity)
    )
    Mesh Scale Mode=::NewEnumerator4
    Sprite Size Mode=::NewEnumerator4
    UsePositionOffset=false
    Write Material Random=true
   ) {module=InitializeParticle001}
   &AddVelocity(Velocity=&RandomRangeVector(Minimum=<0 100 150>, Maximum=<-200 400 250>))
   &InheritVelocity(
    Module.Limit Speed=true
    Velocity Limit=50
    Velocity Scale=&RandomRangeVector(Minimum=<0.5 0.5 0.5>, Maximum=<1 1 1>)
   )
   &AddVelocityInCone(
    Cone Angle=50
    Cone Axis=<0 1 1>
    Module.Use Velocity Falloff On Cone Axis=false
    Velocity Strength=&UniformRangedFloat(Minimum=0, Maximum=50)
   )
   &SphereLocation(
    MaskSpawnGroup=false
    Module.Override Randomness=false
    Sphere Origin:NiagaraPosition=&ConvertVectorToPosition(
     Input Position=Linked() {linkedVariable {name=Particles.SkeletalMeshLocation.SampledPosition, type=NiagaraPosition}}
     Vector Is In Absolute Large World Space:NiagaraBool=false
    )
    Sphere Radius=2
    Surface Only=false
   ) {module=SphereLocation001}
  ]
  particleUpdateScript [
   &ParticleState()
   &Drag(Use Linear Drag=true, Use Rotational Drag=false)
   &GravityForce(Gravity=<0 0 -720>)
   &CurlNoiseForce(Noise Frequency=100, Noise Strength=50, Pan Noise=false)
   &SolveForcesAndVelocity(Clamp Velocity=false, Limit Acceleration=false)
   &ScaleColor(
    Scale Alpha=1.2
    Scale RGB=&MakeVectorFromLinearColorRGB(Color=&ColorFromCurve(ColorCurve=DataInterface() {propertyValues {
        AlphaCurve {keys []}
        BlueCurve {keys [<at 0.0081 0.6244>,<at 0.1423 0.6305>,<at 0.9559 1>]}
        GreenCurve {keys [<at 0.0081 0.8353>,<at 0.1423 0.7629>,<at 0.9559 0.8821>]}
        RedCurve {keys [<at 0.0081 1>,<at 0.1423 1>,<at 0.9559 0.8145>]}
       }}))
    ScaleA=true
    ScaleRGB=true
   )
   &UpdateMeshOrientation(Rotation Rate=&RandomRangeFloat(Minimum=-0.4, Maximum=0.4), Rotation Vector=<1 1 1>)
   &ScaleMeshSize(Scale Factor=&VectorFromFloat(Value=&FloatFromCurve(FloatCurve=DataInterface() {propertyValues {Curve {keys [
          {arriveTangent=8.5765, interpMode=RCIM_Cubic, leaveTangent=8.5765, tangentMode=RCTM_User, time=0.1, value=0.7}
          {arriveTangent=0.65, interpMode=RCIM_Cubic, leaveTangent=0.65, time=0.2, value=1}
          {arriveTangent=-1.25, interpMode=RCIM_Cubic, leaveTangent=-1.25, time=0.6, value=0.625}
          {arriveTangent=0.095, interpMode=RCIM_Cubic, leaveTangent=0.095, tangentMode=RCTM_User, time=1, value=0}
         ]}}})))
   &ScaleSpriteSizeBySpeed(
    Max Scale Factor=&Vector2DFromFloat(Value=1)
    Min Scale Factor=&Vector2DFromFloat(Value=0.25)
    Velocity Threshold=200
   )
  ]
  rendererClasses [&NiagaraMeshRendererProperties]
  renderers [{rendererClass=&NiagaraMeshRendererProperties, rendererIndex=0}]
 }
 Snow_Smoke {
  simTarget=GPUComputeSim
  emitterUpdateScript [
   &EmitterState()
   &SpawnPerUnit(
    Module.Use Max Movement Threshold=true
    Module.Use Movement Tolerance=true
    Movement Tolerance:NiagaraFloat=$User.MovementTolerance
    Spawn Per Unit:NiagaraFloat=&MakeCustomFloatFromBool(
     Bool=$User.bIsFrontEndPreview
     True Float=0
     False Float=&Multiply_Float(
      A=&MakeCustomFloatFromBool(Bool=$User.bIsInWater, True Float=0, False Float=1)
      B=&SelectFloatFromArray(Direct Array Index=$Engine.QualityLevel, Float Selection Array=DataInterface() {propertyValues {FloatData [0,0,1,0.75,0.75]}})
     )
    )
    Use Spawn Probability=false
   )
   &SpawnRate(
    SpawnRate=&MakeCustomFloatFromBool(Bool=$User.bIsFrontEndPreview, True Float=130, False Float=0)
    Use Spawn Probability=false
   )
  ]
  particleSpawnScript [
   &SkeletalMeshLocation(MaskSpawnGroup=false, Mesh Sampling Type=::NewEnumerator5, Skeletal Mesh=$System.Skeletal Mesh, Socket Sampling Mode=::NewEnumerator5)
   &InitializeParticle(
    Lifetime=&UniformRangedFloat(Minimum=0.5, Maximum=0.6)
    Mass=10
    Mass Mode=::NewEnumerator1
    Sprite Rotation Mode=::NewEnumerator2
    Sprite Size Mode=::NewEnumerator1
    Uniform Sprite Size Max=30
    Uniform Sprite Size Min=20
    UsePositionOffset=false
    Write Material Random=true
   ) {module=InitializeParticle001}
   &AddVelocity(Velocity=&RandomRangeVector(Minimum=<0 100 150>, Maximum=<-250 400 250>))
   &InheritVelocity(Module.Limit Speed=true, Velocity Scale=&RandomRangeVector(Minimum=<0.5 0.5 0.5>, Maximum=<1 1 1>))
   &AddVelocityInCone(
    Cone Angle=50
    Cone Axis=<0 -1 1>
    Module.Use Velocity Falloff On Cone Axis=false
    Velocity Strength=&UniformRangedFloat(Minimum=0, Maximum=100)
   )
   &SphereLocation(
    MaskSpawnGroup=false
    Module.Override Randomness=false
    Sphere Origin:NiagaraPosition=&ConvertVectorToPosition(
     Input Position=Linked() {linkedVariable {name=Particles.SkeletalMeshLocation.SampledPosition, type=NiagaraPosition}}
     Vector Is In Absolute Large World Space:NiagaraBool=false
    )
    Sphere Radius=0
    Surface Only=false
   ) {module=SphereLocation001}
  ]
  particleUpdateScript [
   &ParticleState()
   &Drag(Drag=0.5, Use Linear Drag=true, Use Rotational Drag=false)
   &GravityForce(Gravity=&MakeVector(X=0, Y=0, Z=-1200)) {module=GravityForce001}
   &CurlNoiseForce(Noise Frequency=100, Noise Strength=100, Pan Noise=false)
   &SolveForcesAndVelocity(Clamp Velocity=false, Limit Acceleration=false)
   &ScaleColor(
    Scale Alpha=&FloatFromCurve(
     FloatCurve=DataInterface() {propertyValues {Curve {keys [{interpMode=RCIM_Cubic, tangentMode=RCTM_User, time=0, value=1}, {interpMode=RCIM_Cubic, tangentMode=RCTM_User, time=1, value=0}]}}}
     Scale Curve=1.2
    )
    ScaleA=true
    ScaleRGB=true
   )
   &ScaleSpriteSize(Uniform Curve Sprite Scale=DataInterface() {propertyValues {Curve {keys [
        {arriveTangent=-0.9671, interpMode=RCIM_Cubic, leaveTangent=-0.9671, tangentMode=RCTM_User, time=0, value=0.5}
        {arriveTangent=5.2801, interpMode=RCIM_Cubic, leaveTangent=5.2801, tangentMode=RCTM_User, time=0.3, value=1.125}
        <at 0.8 2 RCIM_Cubic>
       ]}}})
   &SpriteRotationRate(Rotation Rate=&RandomRangeFloat(Minimum=-360, Maximum=360))
   &DynamicMaterialParameters(
    Index 0 Param 1=&Add_Float(
     A=&RandomRangeFloat(Minimum=0, Maximum=1)
     B=&FloatFromCurve(FloatCurve=DataInterface() {propertyValues {Curve {keys [
          {arriveTangent=2.739, interpMode=RCIM_Cubic, leaveTangent=2.739, tangentMode=RCTM_User, time=0, value=0}
          {arriveTangent=1.1179, interpMode=RCIM_Cubic, leaveTangent=1.1179, tangentMode=RCTM_User, time=0.25, value=0.51}
          {arriveTangent=0.0224, interpMode=RCIM_Cubic, leaveTangent=0.0224, tangentMode=RCTM_User, time=1, value=1}
         ]}}})
    )
    Index 0 Param 3=&FloatFromCurve(FloatCurve=DataInterface() {propertyValues {Curve {keys [<at 0.25 1>,<at 1 0>]}}}, Scale Curve=0.1)
    Index 0 Param 4=&FloatFromCurve(FloatCurve=DataInterface() {propertyValues {Curve {keys [<at 0 1>,<at 1 0>]}}})
    Index 2 Param 1=&FloatFromCurve(FloatCurve=DataInterface() {propertyValues {Curve {keys [
         {arriveTangent=2.6659, interpMode=RCIM_Cubic, leaveTangent=2.6659, tangentMode=RCTM_User, time=0, value=0}
         {interpMode=RCIM_Cubic, tangentMode=RCTM_User, time=1, value=1}
        ]}}})
    Index 2 Param 2=&FloatFromCurve(
     FloatCurve=DataInterface() {propertyValues {Curve {keys [{tangentMode=RCTM_User, time=0, value=0.2}, {tangentMode=RCTM_User, time=1, value=1.4}]}}}
     Scale Curve=&RandomRangeFloat(Minimum=0.5, Maximum=1)
    )
    Index 2 Param 3=&FloatFromCurve(FloatCurve=DataInterface() {propertyValues {Curve {keys [
         {arriveTangent=2.9718, interpMode=RCIM_Cubic, leaveTangent=2.9718, tangentMode=RCTM_User, time=0, value=1}
         {interpMode=RCIM_Cubic, tangentMode=RCTM_User, time=1, value=2}
        ]}}})
    Index 2 Param 4=&FloatFromCurve(
     FloatCurve=DataInterface() {propertyValues {Curve {keys [
         {arriveTangent=-0.6326, interpMode=RCIM_Cubic, leaveTangent=-0.6326, tangentMode=RCTM_User, time=0, value=1}
         {interpMode=RCIM_Cubic, tangentMode=RCTM_User, time=1, value=0.75}
        ]}}}
     Scale Curve=1.6
    )
    Module.Index0Param0Write=true
    Module.Index0Param1Write=true
    Module.Index0Param2Write=true
    Module.Index0Param3Write=true
    Module.Index1Param0Write=true
    Module.Index1Param1Write=true
    Module.Index1Param2Write=true
    Module.Index1Param3Write=true
    Module.Index2Param0Write=true
    Module.Index2Param1Write=true
    Module.Index2Param2Write=true
    Module.Index2Param3Write=true
    Write Parameter Index 1=true
    Write Parameter Index 2=true
   )
  ]
  rendererClasses [&NiagaraSpriteRendererProperties]
  renderers [{rendererClass=&NiagaraSpriteRendererProperties}]
 }
 Snow_Smoke_CPU {
  emitterUpdateScript [
   &EmitterState()
   &SpawnPerUnit(
    Module.Use Max Movement Threshold=true
    Module.Use Movement Tolerance=true
    Movement Tolerance:NiagaraFloat=$User.MovementTolerance
    Spawn Per Unit:NiagaraFloat=&MakeCustomFloatFromBool(
     Bool=$User.bIsFrontEndPreview
     True Float=0
     False Float=&Multiply_Float(
      A=&MakeCustomFloatFromBool(Bool=$User.bIsInWater, True Float=0, False Float=1)
      B=&SelectFloatFromArray(Direct Array Index=$Engine.QualityLevel, Float Selection Array=DataInterface() {propertyValues {FloatData [1.5,1.5,0,0,0]}})
     )
    )
    Use Spawn Probability=false
   )
   &SpawnRate(
    SpawnRate=&MakeCustomFloatFromBool(Bool=$User.bIsFrontEndPreview, True Float=80, False Float=0)
    Use Spawn Probability=false
   )
  ]
  particleSpawnScript [
   &SkeletalMeshLocation(MaskSpawnGroup=false, Mesh Sampling Type=::NewEnumerator5, Skeletal Mesh=$System.Skeletal Mesh, Socket Sampling Mode=::NewEnumerator5)
   &InitializeParticle(
    Lifetime=&UniformRangedFloat(Minimum=0.5, Maximum=0.6)
    Mass=10
    Mass Mode=::NewEnumerator1
    Sprite Rotation Mode=::NewEnumerator2
    Sprite Size Mode=::NewEnumerator1
    Uniform Sprite Size Max=30
    Uniform Sprite Size Min=20
    UsePositionOffset=false
    Write Material Random=true
   ) {module=InitializeParticle001}
   &AddVelocity(Velocity=&RandomRangeVector(Minimum=<0 100 150>, Maximum=<-250 400 250>))
   &InheritVelocity(Module.Limit Speed=true, Velocity Scale=&RandomRangeVector(Minimum=<0.5 0.5 0.5>, Maximum=<1 1 1>))
   &AddVelocityInCone(
    Cone Angle=50
    Cone Axis=<0 -1 1>
    Module.Use Velocity Falloff On Cone Axis=false
    Velocity Strength=&UniformRangedFloat(Minimum=0, Maximum=100)
   )
   &SphereLocation(
    MaskSpawnGroup=false
    Module.Override Randomness=false
    Sphere Origin:NiagaraPosition=&ConvertVectorToPosition(
     Input Position=Linked() {linkedVariable {name=Particles.SkeletalMeshLocation.SampledPosition, type=NiagaraPosition}}
     Vector Is In Absolute Large World Space:NiagaraBool=false
    )
    Sphere Radius=0
    Surface Only=false
   ) {module=SphereLocation001}
  ]
  particleUpdateScript [
   &ParticleState()
   &Drag(Drag=0.5, Use Linear Drag=true, Use Rotational Drag=false)
   &GravityForce(Gravity=&MakeVector(X=0, Y=0, Z=-1200)) {module=GravityForce001}
   &CurlNoiseForce(Noise Frequency=100, Noise Strength=100, Pan Noise=false)
   &SolveForcesAndVelocity(Clamp Velocity=false, Limit Acceleration=false)
   &ScaleColor(
    Scale Alpha=&FloatFromCurve(
     FloatCurve=DataInterface() {propertyValues {Curve {keys [{interpMode=RCIM_Cubic, tangentMode=RCTM_User, time=0, value=1}, {interpMode=RCIM_Cubic, tangentMode=RCTM_User, time=1, value=0}]}}}
     Scale Curve=1.2
    )
    ScaleA=true
    ScaleRGB=true
   )
   &ScaleSpriteSize(Uniform Curve Sprite Scale=DataInterface() {propertyValues {Curve {keys [
        {arriveTangent=-0.9671, interpMode=RCIM_Cubic, leaveTangent=-0.9671, tangentMode=RCTM_User, time=0, value=0.5}
        {arriveTangent=5.2801, interpMode=RCIM_Cubic, leaveTangent=5.2801, tangentMode=RCTM_User, time=0.3, value=1.125}
        <at 0.8 2 RCIM_Cubic>
       ]}}})
   &SpriteRotationRate(Rotation Rate=&RandomRangeFloat(Minimum=-360, Maximum=360))
   &DynamicMaterialParameters(
    Index 0 Param 1=&Add_Float(
     A=&RandomRangeFloat(Minimum=0, Maximum=1)
     B=&FloatFromCurve(FloatCurve=DataInterface() {propertyValues {Curve {keys [
          {arriveTangent=2.739, interpMode=RCIM_Cubic, leaveTangent=2.739, tangentMode=RCTM_User, time=0, value=0}
          {arriveTangent=1.1179, interpMode=RCIM_Cubic, leaveTangent=1.1179, tangentMode=RCTM_User, time=0.25, value=0.51}
          {arriveTangent=0.0224, interpMode=RCIM_Cubic, leaveTangent=0.0224, tangentMode=RCTM_User, time=1, value=1}
         ]}}})
    )
    Index 0 Param 3=&FloatFromCurve(FloatCurve=DataInterface() {propertyValues {Curve {keys [<at 0.25 1>,<at 1 0>]}}}, Scale Curve=0.1)
    Index 0 Param 4=&FloatFromCurve(FloatCurve=DataInterface() {propertyValues {Curve {keys [<at 0 1>,<at 1 0>]}}})
    Index 2 Param 1=&FloatFromCurve(FloatCurve=DataInterface() {propertyValues {Curve {keys [
         {arriveTangent=2.6659, interpMode=RCIM_Cubic, leaveTangent=2.6659, tangentMode=RCTM_User, time=0, value=0}
         {interpMode=RCIM_Cubic, tangentMode=RCTM_User, time=1, value=1}
        ]}}})
    Index 2 Param 2=&FloatFromCurve(
     FloatCurve=DataInterface() {propertyValues {Curve {keys [{tangentMode=RCTM_User, time=0, value=0.2}, {tangentMode=RCTM_User, time=1, value=1.4}]}}}
     Scale Curve=&RandomRangeFloat(Minimum=0.5, Maximum=1)
    )
    Index 2 Param 3=&FloatFromCurve(FloatCurve=DataInterface() {propertyValues {Curve {keys [
         {arriveTangent=2.9718, interpMode=RCIM_Cubic, leaveTangent=2.9718, tangentMode=RCTM_User, time=0, value=1}
         {interpMode=RCIM_Cubic, tangentMode=RCTM_User, time=1, value=2}
        ]}}})
    Index 2 Param 4=&FloatFromCurve(
     FloatCurve=DataInterface() {propertyValues {Curve {keys [
         {arriveTangent=-0.6326, interpMode=RCIM_Cubic, leaveTangent=-0.6326, tangentMode=RCTM_User, time=0, value=1}
         {interpMode=RCIM_Cubic, tangentMode=RCTM_User, time=1, value=0.75}
        ]}}}
     Scale Curve=1.6
    )
    Module.Index0Param0Write=true
    Module.Index0Param1Write=true
    Module.Index0Param2Write=true
    Module.Index0Param3Write=true
    Module.Index1Param0Write=true
    Module.Index1Param1Write=true
    Module.Index1Param2Write=true
    Module.Index1Param3Write=true
    Module.Index2Param0Write=true
    Module.Index2Param1Write=true
    Module.Index2Param2Write=true
    Module.Index2Param3Write=true
    Write Parameter Index 1=true
    Write Parameter Index 2=true
   )
  ]
  rendererClasses [&NiagaraSpriteRendererProperties]
  renderers [{rendererClass=&NiagaraSpriteRendererProperties}]
 }
 Snow_Spray {
  simTarget=GPUComputeSim
  emitterUpdateScript [
   &EmitterState()
   &SpawnPerUnit(
    Module.Use Max Movement Threshold=true
    Module.Use Movement Tolerance=true
    Movement Tolerance:NiagaraFloat=$User.MovementTolerance
    Spawn Per Unit:NiagaraFloat=&MakeCustomFloatFromBool(
     Bool=$User.bIsFrontEndPreview
     True Float=0
     False Float=&Multiply_Float(
      A=&MakeCustomFloatFromBool(Bool=$User.bIsInWater, True Float=0, False Float=1)
      B=&SelectFloatFromArray(Direct Array Index=$Engine.QualityLevel, Float Selection Array=DataInterface() {propertyValues {FloatData [0,0,1.5,1,1]}})
     )
    )
    Use Spawn Probability=false
   )
   &SpawnRate(
    SpawnRate=&MakeCustomFloatFromBool(Bool=$User.bIsFrontEndPreview, True Float=25, False Float=0)
    Use Spawn Probability=false
   )
  ]
  particleSpawnScript [
   &SkeletalMeshLocation(MaskSpawnGroup=false, Mesh Sampling Type=::NewEnumerator5, Skeletal Mesh=$System.Skeletal Mesh, Socket Sampling Mode=::NewEnumerator5)
   &InitializeParticle(
    Color=<rgba 2 2 2 1>
    Lifetime=&UniformRangedFloat(Minimum=0.15, Maximum=0.3)
    Mass=10
    Mass Mode=::NewEnumerator1
    Sprite Rotation Mode=::NewEnumerator2
    Sprite Size Mode=::NewEnumerator1
    Uniform Sprite Size Max=20
    Uniform Sprite Size Min=10
    UsePositionOffset=false
    Write Material Random=true
   ) {module=InitializeParticle001}
   &AddVelocity(Velocity=&RandomRangeVector(Minimum=<0 100 150>, Maximum=<-250 400 250>))
   &InheritVelocity(Module.Limit Speed=true, Velocity Scale=&RandomRangeVector(Minimum=<0.5 0.5 0.5>, Maximum=<1 1 1>))
   &AddVelocityInCone(
    Cone Angle=50
    Cone Axis=<0 -1 1>
    Module.Use Velocity Falloff On Cone Axis=false
    Velocity Strength=&UniformRangedFloat(Minimum=0, Maximum=100)
   )
   &SphereLocation(
    MaskSpawnGroup=false
    Module.Override Randomness=false
    Sphere Origin:NiagaraPosition=&ConvertVectorToPosition(
     Input Position=Linked() {linkedVariable {name=Particles.SkeletalMeshLocation.SampledPosition, type=NiagaraPosition}}
     Vector Is In Absolute Large World Space:NiagaraBool=false
    )
    Sphere Radius=0
    Surface Only=false
   ) {module=SphereLocation001}
  ]
  particleUpdateScript [
   &ParticleState()
   &Drag(Use Linear Drag=true, Use Rotational Drag=false)
   &GravityForce(Gravity=<0 0 -720>)
   &CurlNoiseForce(Noise Frequency=100, Noise Strength=100, Pan Noise=false)
   &SolveForcesAndVelocity(Clamp Velocity=false, Limit Acceleration=false)
   &ScaleColor(
    Scale Alpha=1.2
    Scale RGB=&MakeVectorFromLinearColorRGB(Color=&ColorFromCurve(ColorCurve=DataInterface() {propertyValues {
        AlphaCurve {keys []}
        BlueCurve {keys [<at 0.0081 0.6244>,<at 0.1423 0.6305>,<at 0.9559 1>]}
        GreenCurve {keys [<at 0.0081 0.8353>,<at 0.1423 0.7629>,<at 0.9559 0.8821>]}
        RedCurve {keys [<at 0.0081 1>,<at 0.1423 1>,<at 0.9559 0.8145>]}
       }}))
    ScaleA=true
    ScaleRGB=true
   )
   &ScaleSpriteSize(Uniform Curve Sprite Scale=DataInterface() {propertyValues {Curve {keys [
        {arriveTangent=8.5765, interpMode=RCIM_Cubic, leaveTangent=8.5765, tangentMode=RCTM_User, time=0.1, value=0.5}
        {arriveTangent=-0.5096, interpMode=RCIM_Cubic, leaveTangent=-0.5096, time=0.2, value=1}
        {arriveTangent=-5.3839, interpMode=RCIM_Cubic, leaveTangent=-5.3839, tangentMode=RCTM_User, time=0.7622, value=0.1625}
        {arriveTangent=0.095, interpMode=RCIM_Cubic, leaveTangent=0.095, tangentMode=RCTM_User, time=0.8, value=-0.05}
       ]}}})
   &SpriteRotationRate(Rotation Rate=&RandomRangeFloat(Minimum=-360, Maximum=360))
  ]
  rendererClasses [&NiagaraSpriteRendererProperties]
  renderers [{rendererClass=&NiagaraSpriteRendererProperties}]
 }
 Snow_Spray_CPU {
  emitterUpdateScript [
   &EmitterState()
   &SpawnPerUnit(
    Module.Use Max Movement Threshold=true
    Module.Use Movement Tolerance=true
    Movement Tolerance:NiagaraFloat=$User.MovementTolerance
    Spawn Per Unit:NiagaraFloat=&MakeCustomFloatFromBool(
     Bool=$User.bIsFrontEndPreview
     True Float=0
     False Float=&Multiply_Float(
      A=&MakeCustomFloatFromBool(Bool=$User.bIsInWater, True Float=0, False Float=1)
      B=&SelectFloatFromArray(Direct Array Index=$Engine.QualityLevel, Float Selection Array=DataInterface() {propertyValues {FloatData [3,2.5,0,0,0]}})
     )
    )
    Use Spawn Probability=false
   )
   &SpawnRate(
    SpawnRate=&MakeCustomFloatFromBool(Bool=$User.bIsFrontEndPreview, True Float=25, False Float=0)
    Use Spawn Probability=false
   )
  ]
  particleSpawnScript [
   &SkeletalMeshLocation(MaskSpawnGroup=false, Mesh Sampling Type=::NewEnumerator5, Skeletal Mesh=$System.Skeletal Mesh, Socket Sampling Mode=::NewEnumerator5)
   &InitializeParticle(
    Color=<rgba 2 2 2 1>
    Lifetime=&UniformRangedFloat(Minimum=0.1, Maximum=0.15)
    Mass=10
    Mass Mode=::NewEnumerator1
    Sprite Rotation Mode=::NewEnumerator2
    Sprite Size Mode=::NewEnumerator1
    Uniform Sprite Size Max=25
    Uniform Sprite Size Min=20
    UsePositionOffset=false
    Write Material Random=true
   ) {module=InitializeParticle001}
   &AddVelocity(Velocity=&RandomRangeVector(Minimum=<-250 250 150>, Maximum=<-500 500 250>)) {module=AddVelocity002}
   &InheritVelocity(Module.Limit Speed=true, Velocity Scale=&RandomRangeVector(Minimum=<0.5 0.5 0.5>, Maximum=<1 1 1>))
   &AddVelocityInCone(
    Cone Angle=50
    Cone Axis=<0 -1 1>
    Module.Use Velocity Falloff On Cone Axis=false
    Velocity Strength=&UniformRangedFloat(Minimum=200, Maximum=300)
   )
   &SphereLocation(
    MaskSpawnGroup=false
    Module.Override Randomness=false
    Sphere Origin:NiagaraPosition=&ConvertVectorToPosition(
     Input Position=Linked() {linkedVariable {name=Particles.SkeletalMeshLocation.SampledPosition, type=NiagaraPosition}}
     Vector Is In Absolute Large World Space:NiagaraBool=false
    )
    Sphere Radius=0
    Surface Only=false
   ) {module=SphereLocation001}
  ]
  particleUpdateScript [
   &ParticleState()
   &Drag(Drag=4, Use Linear Drag=true, Use Rotational Drag=false)
   &GravityForce(Gravity=<0 0 -720>)
   &CurlNoiseForce(Noise Frequency=100, Noise Strength=100, Pan Noise=false)
   &SolveForcesAndVelocity(Clamp Velocity=false, Limit Acceleration=false)
   &ScaleColor(
    Scale Alpha=1.2
    Scale RGB=&MakeVectorFromLinearColorRGB(Color=&ColorFromCurve(ColorCurve=DataInterface() {propertyValues {
        AlphaCurve {keys []}
        BlueCurve {keys [<at 0.0081 0.6244>,<at 0.1423 0.6305>,<at 0.9559 1>]}
        GreenCurve {keys [<at 0.0081 0.8353>,<at 0.1423 0.7629>,<at 0.9559 0.8821>]}
        RedCurve {keys [<at 0.0081 1>,<at 0.1423 1>,<at 0.9559 0.8145>]}
       }}))
    ScaleA=true
    ScaleRGB=true
   )
   &ScaleSpriteSize(Uniform Curve Sprite Scale=DataInterface() {propertyValues {Curve {keys [
        {arriveTangent=-0.2487, interpMode=RCIM_Cubic, leaveTangent=-0.2487, tangentMode=RCTM_User, time=0, value=0.1}
        {arriveTangent=7.0112, interpMode=RCIM_Cubic, leaveTangent=0.023, tangentMode=RCTM_Break, time=0.2, value=1}
        {arriveTangent=-5.3839, interpMode=RCIM_Cubic, leaveTangent=-5.3839, tangentMode=RCTM_User, time=0.7622, value=0.1625}
        {arriveTangent=0.095, interpMode=RCIM_Cubic, leaveTangent=0.095, tangentMode=RCTM_User, time=0.8, value=-0.05}
       ]}}}) {module=ScaleSpriteSize001}
   &SpriteRotationRate(Rotation Rate=&RandomRangeFloat(Minimum=-360, Maximum=360))
  ]
  rendererClasses [&NiagaraSpriteRendererProperties]
  renderers [{rendererClass=&NiagaraSpriteRendererProperties}]
 }
 Snowflakes {
  simTarget=GPUComputeSim
  emitterUpdateScript [
   &EmitterState()
   &SpawnPerUnit(
    Module.Use Max Movement Threshold=true
    Module.Use Movement Tolerance=true
    Movement Tolerance:NiagaraFloat=$User.MovementTolerance
    Spawn Per Unit:NiagaraFloat=&MakeCustomFloatFromBool(
     Bool=$User.bIsFrontEndPreview
     True Float=0
     False Float=&Multiply_Float(
      A=&MakeCustomFloatFromBool(Bool=$User.bIsInWater, True Float=0, False Float=1)
      B=&SelectFloatFromArray(Direct Array Index=$Engine.QualityLevel, Float Selection Array=DataInterface() {propertyValues {FloatData [0,0,6,5,5]}})
     )
    )
    Use Spawn Probability=false
   )
   &SpawnRate(
    SpawnRate=&MakeCustomFloatFromBool(Bool=$User.bIsFrontEndPreview, True Float=10, False Float=0)
    Use Spawn Probability=false
   ) {module=SpawnRate001}
  ]
  particleSpawnScript [
   &InitializeParticle(
    Color=&RandomRangeLinearColor(Minimum=<rgba 0.9 0.9667 1 0.05>, Maximum=<rgba 0.9 0.9667 1 0.075>)
    Lifetime Max=2.5
    Lifetime Min=1.5
    Lifetime Mode=::NewEnumerator1
    Sprite Rotation Mode=::NewEnumerator2
    Sprite Size Mode=::NewEnumerator1
    Uniform Sprite Size Max=7
    UsePositionOffset=false
    Write Material Random=true
   )
   &SkeletalMeshLocation(
    Inherit Velocity=false
    MaskSpawnGroup=false
    Mesh Sampling Type=::NewEnumerator5
    Skeletal Mesh=$System.Skeletal Mesh
    Socket Sampling Mode=::NewEnumerator5
   )
   &ShapeLocation(MaskSpawnGroup=false, Override Seed=false, Shape Origin=$Output.SkeletalMeshLocation.SampledPosition, Sphere Radius=12, UseOwnerScale=true)
   &AddVelocity(Velocity=&RandomRangeVector(Minimum=<-200 550 150>, Maximum=<-350 650 250>)) {module=AddVelocity002}
   &InheritVelocity(Module.Limit Speed=true, Velocity Scale=&RandomRangeVector(Minimum=<0.5 0.5 0.5>, Maximum=<1 1 1>))
   &AddVelocityInCone(
    Cone Angle=50
    Cone Axis=<0 -1 1>
    Module.Use Velocity Falloff On Cone Axis=false
    Velocity Strength=&UniformRangedFloat(Minimum=200, Maximum=300)
   )
   /SetVariables(Particles.RandFacing:Vector3f=&RandomRangeVector(Minimum=<-1 -1 -1>, Maximum=<1 1 1>)) {isSetParameters=true, module=SetVariables_#4}
   &CameraOffset(Camera Offset Amount=6)
   &SubUVAnimation(
    End Frame Range Override=3
    Evaluation Type=::NewEnumerator0
    "Sprite Renderer "=DataInterface() {propertyValues {SpriteRenderer=&self:Snowflakes_5.NiagaraSpriteRendererProperties_2}}
    SubUV Animation Mode:ENiagara_SubUVLookupModeV2=::NewEnumerator2
    UseEndFrame=true
    UseStartFrame=true
   )
  ]
  particleUpdateScript [
   &ParticleState()
   &CurlNoiseForce(
    Noise Frequency=33
    Noise Strength=&RandomRangeFloat(Minimum=400, Maximum=800)
    Pan Noise=true
    Pan Noise Field=<0.3 0.3 0.3>
   )
   &GravityForce(Gravity=&MakeVector(X=0, Y=0, Z=-450)) {module=GravityForce001}
   &Drag(Drag=&RandomRangeFloat(Minimum=4.5, Maximum=5), Use Linear Drag=true, Use Rotational Drag=false)
   &SolveForcesAndVelocity(Clamp Velocity=false, Limit Acceleration=false)
   &ScaleColor(
    Scale Alpha=&FloatFromCurve(FloatCurve=DataInterface() {propertyValues {Curve {keys [
         {interpMode=RCIM_Cubic, tangentMode=RCTM_User, time=0, value=0}
         {interpMode=RCIM_Cubic, tangentMode=RCTM_User, time=0.1, value=1}
         {interpMode=RCIM_Cubic, tangentMode=RCTM_User, time=1, value=0}
        ]}}})
    Scale RGB=&MakeVectorFromLinearColorRGB(Color=&ColorFromCurve(ColorCurve=DataInterface() {propertyValues {
        AlphaCurve {keys []}
        BlueCurve {keys [<at 0.2426 1 RCIM_Cubic>,<at 1 1 RCIM_Cubic>]}
        GreenCurve {keys [<at 0.2426 1 RCIM_Cubic>,<at 1 0.755 RCIM_Cubic>]}
        RedCurve {keys [<at 0.2426 1 RCIM_Cubic>,<at 1 0.65 RCIM_Cubic>]}
       }}))
    ScaleA=true
    ScaleRGB=true
   )
   &ScaleSpriteSize(Scale Factor=&Vector2DFromFloat(Value=&FloatFromCurve(FloatCurve=DataInterface() {propertyValues {Curve {keys [
          <at 0 0 RCIM_Cubic>
          {arriveTangent=2, interpMode=RCIM_Cubic, leaveTangent=2, time=0.1619, value=1}
          {arriveTangent=-1.1932, interpMode=RCIM_Cubic, leaveTangent=-1.1932, time=0.5, value=1}
          <at 1 0 RCIM_Cubic>
         ]}}})))
  ]
  rendererClasses [&NiagaraSpriteRendererProperties]
  renderers [{rendererClass=&NiagaraSpriteRendererProperties}]
 }
 Snowflakes_CPU {
  emitterUpdateScript [
   &EmitterState()
   &SpawnPerUnit(
    Module.Use Max Movement Threshold=true
    Module.Use Movement Tolerance=true
    Movement Tolerance:NiagaraFloat=$User.MovementTolerance
    Spawn Per Unit:NiagaraFloat=&MakeCustomFloatFromBool(
     Bool=$User.bIsFrontEndPreview
     True Float=0
     False Float=&Multiply_Float(
      A=&MakeCustomFloatFromBool(Bool=$User.bIsInWater, True Float=0, False Float=1)
      B=&SelectFloatFromArray(Direct Array Index=$Engine.QualityLevel, Float Selection Array=DataInterface() {propertyValues {FloatData [10,8,0,0,0]}})
     )
    )
    Use Spawn Probability=false
   )
   &SpawnRate(
    SpawnRate=&MakeCustomFloatFromBool(Bool=$User.bIsFrontEndPreview, True Float=10, False Float=0)
    Use Spawn Probability=false
   ) {module=SpawnRate001}
  ]
  particleSpawnScript [
   &InitializeParticle(
    Color=<rgba 0.9 0.9667 1 0.1>
    Lifetime Max=2.5
    Lifetime Min=1.5
    Lifetime Mode=::NewEnumerator1
    Sprite Rotation Mode=::NewEnumerator2
    Sprite Size Mode=::NewEnumerator1
    Uniform Sprite Size Max=7
    UsePositionOffset=false
    Write Material Random=true
   )
   &SkeletalMeshLocation(
    Inherit Velocity=false
    MaskSpawnGroup=false
    Mesh Sampling Type=::NewEnumerator5
    Skeletal Mesh=$System.Skeletal Mesh
    Socket Sampling Mode=::NewEnumerator5
   )
   &ShapeLocation(MaskSpawnGroup=false, Override Seed=false, Shape Origin=$Output.SkeletalMeshLocation.SampledPosition, Sphere Radius=12, UseOwnerScale=true)
   &AddVelocity(Velocity=&RandomRangeVector(Minimum=<-200 550 150>, Maximum=<-350 650 250>)) {module=AddVelocity003}
   &InheritVelocity(Module.Limit Speed=true, Velocity Scale=&RandomRangeVector(Minimum=<0.5 0.5 0.5>, Maximum=<1 1 1>))
   &AddVelocityInCone(
    Cone Angle=50
    Cone Axis=<0 -1 1>
    Module.Use Velocity Falloff On Cone Axis=false
    Velocity Strength=&UniformRangedFloat(Minimum=200, Maximum=300)
   )
   /SetVariables(Particles.RandFacing:Vector3f=&RandomRangeVector(Minimum=<-1 -1 -1>, Maximum=<1 1 1>)) {isSetParameters=true, module=SetVariables_#4}
   &CameraOffset(Camera Offset Amount=6)
   &SubUVAnimation(
    End Frame Range Override=3
    Evaluation Type=::NewEnumerator0
    "Sprite Renderer "=DataInterface() {propertyValues {SpriteRenderer=&self:Snowflakes_CPU_1.NiagaraSpriteRendererProperties_2}}
    SubUV Animation Mode:ENiagara_SubUVLookupModeV2=::NewEnumerator2
    UseEndFrame=true
    UseStartFrame=true
   )
  ]
  particleUpdateScript [
   &ParticleState()
   &CurlNoiseForce(
    Noise Frequency=33
    Noise Strength=&RandomRangeFloat(Minimum=400, Maximum=800)
    Pan Noise=true
    Pan Noise Field=<0.3 0.3 0.3>
   )
   &GravityForce(Gravity=&MakeVector(X=0, Y=0, Z=-450)) {module=GravityForce001}
   &Drag(Drag=&RandomRangeFloat(Minimum=4.5, Maximum=5), Use Linear Drag=true, Use Rotational Drag=false)
   &SolveForcesAndVelocity(Clamp Velocity=false, Limit Acceleration=false)
   &ScaleColor(
    Scale Alpha=&FloatFromCurve(FloatCurve=DataInterface() {propertyValues {Curve {keys [
         {interpMode=RCIM_Cubic, tangentMode=RCTM_User, time=0, value=0}
         {interpMode=RCIM_Cubic, tangentMode=RCTM_User, time=0.1, value=1}
         {interpMode=RCIM_Cubic, tangentMode=RCTM_User, time=1, value=0}
        ]}}})
    Scale RGB=&MakeVectorFromLinearColorRGB(Color=&ColorFromCurve(ColorCurve=DataInterface() {propertyValues {
        AlphaCurve {keys []}
        BlueCurve {keys [<at 0.2426 1 RCIM_Cubic>,<at 1 1 RCIM_Cubic>]}
        GreenCurve {keys [<at 0.2426 1 RCIM_Cubic>,<at 1 0.755 RCIM_Cubic>]}
        RedCurve {keys [<at 0.2426 1 RCIM_Cubic>,<at 1 0.65 RCIM_Cubic>]}
       }}))
    ScaleA=true
    ScaleRGB=true
   )
   &ScaleSpriteSize(Scale Factor=&Vector2DFromFloat(Value=&FloatFromCurve(FloatCurve=DataInterface() {propertyValues {Curve {keys [
          <at 0 0 RCIM_Cubic>
          {arriveTangent=2, interpMode=RCIM_Cubic, leaveTangent=2, time=0.1619, value=1}
          {arriveTangent=-1.1932, interpMode=RCIM_Cubic, leaveTangent=-1.1932, time=0.5, value=1}
          <at 1 0 RCIM_Cubic>
         ]}}})))
  ]
  rendererClasses [&NiagaraSpriteRendererProperties]
  renderers [{rendererClass=&NiagaraSpriteRendererProperties}]
 }
}
summary {
 systemName=NS_SleetEmote_SnowSpoutFX
 emitters [
  {emitterName=Snow_Chunks, simTarget=GPUComputeSim, rendererClasses [&NiagaraMeshRendererProperties]}
  {emitterName=Snow_Spray, simTarget=GPUComputeSim, rendererClasses [&NiagaraSpriteRendererProperties]}
  {emitterName=Collision_Snow, simTarget=GPUComputeSim, rendererClasses [&NiagaraSpriteRendererProperties]}
  {emitterName=Snow_Smoke, simTarget=GPUComputeSim, rendererClasses [&NiagaraSpriteRendererProperties]}
  {emitterName=Snowflakes_CPU, rendererClasses [&NiagaraSpriteRendererProperties]}
  {emitterName=Snow_Spray_CPU, rendererClasses [&NiagaraSpriteRendererProperties]}
  {emitterName=Snow_Smoke_CPU, rendererClasses [&NiagaraSpriteRendererProperties]}
  {emitterName=Collision_Snow_CPU, rendererClasses [&NiagaraSpriteRendererProperties]}
  {emitterName=Snowflakes, simTarget=GPUComputeSim, rendererClasses [&NiagaraSpriteRendererProperties]}
  {emitterName=Snow_Chunks_CPU, rendererClasses [&NiagaraMeshRendererProperties]}
 ]
 userVariables [
  User.bIsFrontEnd {type=NiagaraBool}
  User.bIsFrontEndPreview {type=NiagaraBool}
  User.bIsInWater {type=NiagaraBool}
  User.MovementTolerance {type=NiagaraFloat, defaultValue=1, description=""}
 ]
}
